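; Vehicle cosmetic actor for the ZigZag drift trail, driving a Niagara effect on the vehicle.
; Applying the cosmetic sets the trail system and wheel offset; BeginPlay clears absolute rotation.
; Drift start and landing mid-drift activate it, gated by a significance LOD check, while leaving the ground or ending the drift deactivates it.
; SetupPaintSettings passes paint state and color variant to Niagara and looks up the painted-color data table.
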(event Vehicle|Cosmetics|EventOnCosmeticApplied
  ;[1 1 1] Set Initial Values
  (Niagara|SetNiagaraSystemAsset (Variables|Default|GetNS_DriftTrail) (Variables|Default|GetWheelTrailFX) true)
  (Niagara|SetNiagaraVariable(Vector3) (Variables|Default|GetNS_DriftTrail) "offset" (Variables|Default|GetOffsetFromWheels)))

(event EventBeginPlay
  ;[1 1 1] Set Initial Values
  (Variables|/Script/Engine.SceneComponent|SetbAbsoluteRotation false (Variables|Default|GetNS_DriftTrail)))

(event Vehicle|Cosmetics|EventOnVehicleWheelsLeftGround
  ;[1 1 1] In Air(Jumping) Events
  (if (Components|Activation|IsActive (Variables|Default|GetNS_DriftTrail))
    (Components|Activation|Deactivate (Variables|Default|GetNS_DriftTrail))))

(event Vehicle|Cosmetics|EventOnDriftStarted
  ;[1 1 1] Drifting Events
  (bind _returnvalue (VehicleCosmetics|FX|GetMaxSignificanceLODRearWheelDriftFX))
  (bind _btrigger (Class|BPVehicleCosmeticsLibrary|ShouldTriggerFX (Variables|Vehicle|Cosmetics|GetVehicleCosmeticsInterface) _returnvalue))
  (if _btrigger
    (Components|Activation|SetActive (Variables|Default|GetNS_DriftTrail) true false)))

(event Vehicle|Cosmetics|EventOnVehicleLanded
  (params (LandingForced :type Float%20%28single-precision%29) (bLandedKickflip :type Boolean))
  ;[1 1 1] In Air(Jumping) Events
  (if (Utilities|DoesObjectImplementInterface (Actor|GetOwner) "/Script/VehicleCosmeticsRuntime.VehicleCosmeticsInterface")
    (if (Vehicle|Cosmetics|Drifting|IsDrifting (Variables|Vehicle|Cosmetics|GetVehicleCosmeticsInterface))
      (bind _returnvalue (VehicleCosmetics|FX|GetMaxSignificanceLODRearWheelDriftFX))
      (bind _btrigger (Class|BPVehicleCosmeticsLibrary|ShouldTriggerFX (Variables|Vehicle|Cosmetics|GetVehicleCosmeticsInterface) _returnvalue))
      (if _btrigger
        (Components|Activation|Activate (Variables|Default|GetNS_DriftTrail) false)))))

(event Vehicle|Cosmetics|EventOnDriftFinished
  ;[1 1 1] Drifting Events
  (Components|Activation|SetActive (Variables|Default|GetNS_DriftTrail) false false))

(event Vehicle|Cosmetics|EventOnColorReady
  (params (Color :type Name :default "None") (AdditionalInfo :type Vehicle%20Cosmetics%20Additional%20Parameter%20Infos%20Structure))
  ;[1 1 1] Setup Paint Settings
  (CallFunction|SetupPaintSettings Color AdditionalInfo))

(fn SetupPaintSettings
  (params (Color :type Name) (AdditionalInfo :type Vehicle%20Cosmetics%20Additional%20Parameter%20Infos%20Structure))
  (bind _ns_drifttrail (Variables|Default|GetNS_DriftTrail))
  (if (Utilities|String|EqualExactly(String) (Utilities|String|ToString(Name) Color) "None")
    (Niagara|SetNiagaraVariable(Bool) _ns_drifttrail "bIsPainted" false)
    (else
      (Niagara|SetNiagaraVariable(Bool) _ns_drifttrail "bIsPainted" true)
      (Niagara|SetNiagaraVariable(Float) (Variables|Default|GetNS_DriftTrail) "PaintedScalar" (Variables|Default|GetPaintedScalar))
      (bind _color_index (Class|BPVehicleCosmeticsFunctions|BPGetNSIndexbyColor (Variables|GetColor)))
      (Niagara|SetNiagaraVariable(Int32) (Variables|Default|GetNS_DriftTrail) "ColorVariantID" _color_index)
      (Class|BPVehicleCosmeticsFunctionLibrary|CosmeticActorColorReady (Variables|GetColor) "PaintedColor" (Variables|GetAdditionalInfo) (Variables|Default|GetNS_DriftTrail))
      (bind _returnvalue (Utilities|GetDataTableRow "/VehicleCosmetics/DataTables/DT_VehicleCosmetics_Painted.DT_VehicleCosmetics_Painted" (Variables|GetColor)))
      (return (Utilities|Struct|BreakStructVehicleCosmeticsPainted _returnvalue) (.Balanced_34_F677FC9D4B1B4146E8CC9DBCB9E0943E (Utilities|Struct|BreakStructVehicleCosmeticsPainted _returnvalue)) (.Metallic_41_F57CE68E419E301DD160BEB5D0E32BFC (Utilities|Struct|BreakStructVehicleCosmeticsPainted _returnvalue))))))

(fn ConstructionScript)
