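# Niagara drift trail cosmetic for vehicles, streaming glowing zigzag ribbons and animated ring sprites backward from skeletal mesh attach points.
# The ring emitter merges three spawn groups, switching bone, lifetime, size and velocity per group.
# A ribbon emitter spawns two trails, each drawn by three renderers with extra color and width bindings.
# User flags switch painted colors and frontend-preview tuning, while FN_DisableTireFX masks effects per tire.
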NiagaraSystem /VehicleCosmetics/DriftTrail/DriftTrail_ZigZag/FX/NS_DriftTrail_ZigZag_MP
dependencies {
 usedDataInterfaces [
  &NiagaraDataInterfaceVectorCurve
  &NiagaraDataInterfaceCurve
  &NiagaraDataInterfaceArrayInt32
  &NiagaraDataInterfaceSpriteRendererInfo
  &NiagaraDataInterfaceSkeletalMesh
 ]
 usedDynamicInputs []
 usedModules [
  &self:SystemSpawnScript.SystemScriptSource.SystemScriptGraph.NiagaraNodeAssignment_0.SetVariables_#1
  &SystemState
  &FN_UnpackVehicleFlags
  /Game/Effects/Niagara/Modules/Vehicle/FN_UpdateVehicleAttribs
  /Game/Effects/Niagara/Modules/Vehicle/FN_VehicleFlags_Cheatcodes
 ]
}
emitters {
 NE_VC_Rings_Left001 {
  emitterUpdateScript [
   &EmitterState()
   &SpawnRate(
    SpawnGroup=1
    SpawnRate=&MakeCustomFloatFromBool(Bool=$User.bIsFrontendPreview, True Float=10, False Float=5)
    Use Spawn Probability=false
   )
   &SpawnRate(
    SpawnRate=&MakeCustomFloatFromBool(Bool=$User.bIsFrontendPreview, True Float=10, False Float=5)
    Use Spawn Probability=false
   ) {module=SpawnRate001}
   &SpawnRate(SpawnGroup=2, SpawnRate=20, Use Spawn Probability=false) {module=SpawnRate002}
  ]
  particleSpawnScript [
   /SetVariables(Particles.NewVCTag:NiagaraInt32=&SpawnGroup()) {isSetParameters=true}
   &VC_PaintedVariantsOverride() {enabled=false}
   &InitializeParticle(
    Color=&MakeCustomLinearColorFromBool(Bool=$User.bIsPainted, True Color=$User.PaintedColor, False Color=<rgba 1 1 1 1>)
    Lifetime=/VehicleCosmetics/FX/Niagara/Modules/MergeEmitters/VC_FloatBySpawnGroup(
     SpawnGroup or ExecIndex:NiagaraInt32=0
     CustomInt:NiagaraInt32=0
     MaximumSpawnGroups:NiagaraInt32=3
     Float SG 0:NiagaraFloat=&MakeCustomFloatFromBool(
      Bool=$User.bIsFrontendPreview
      True Float=&RandomRangeFloat(Minimum=0.5, Maximum=0.6)
      False Float=&RandomRangeFloat(Minimum=0.35, Maximum=0.46)
     )
     Float SG 1:NiagaraFloat=&MakeCustomFloatFromBool(
      Bool=$User.bIsFrontendPreview
      True Float=&RandomRangeFloat(Minimum=0.5, Maximum=0.6)
      False Float=&RandomRangeFloat(Minimum=0.35, Maximum=0.46)
     )
     Float SG 2:NiagaraFloat=&RandomRangeFloat(Minimum=0.22, Maximum=0.42)
     Float SG 3:NiagaraFloat=0
     Float SG 4:NiagaraFloat=0
     Float SG 5:NiagaraFloat=0
    )
    Sprite Rotation Mode=::NewEnumerator2
    Sprite Size Mode=::NewEnumerator1
    Uniform Sprite Size Max=&VC_FloatBySpawnGroup(
     SpawnGroup or ExecIndex:NiagaraInt32=0
     CustomInt:NiagaraInt32=0
     MaximumSpawnGroups:NiagaraInt32=3
     Float SG 0:NiagaraFloat=55
     Float SG 1:NiagaraFloat=55
     Float SG 2:NiagaraFloat=40
     Float SG 3:NiagaraFloat=0
     Float SG 4:NiagaraFloat=0
     Float SG 5:NiagaraFloat=0
    )
    Uniform Sprite Size Min=&VC_FloatBySpawnGroup(
     SpawnGroup or ExecIndex:NiagaraInt32=0
     CustomInt:NiagaraInt32=0
     MaximumSpawnGroups:NiagaraInt32=3
     Float SG 0:NiagaraFloat=40
     Float SG 1:NiagaraFloat=40
     Float SG 2:NiagaraFloat=35
     Float SG 3:NiagaraFloat=0
     Float SG 4:NiagaraFloat=0
     Float SG 5:NiagaraFloat=0
    )
    UsePositionOffset=false
    Write Material Random=true
   )
   &PartitionParticles(Number Of Partitions=&MaskIntBySpawnGroup(Masked Int=2, Passthrough Int=0, Spawn Group Mask=2))
   &SkeletalMeshLocation(
    Bone / Socket Index=&Add_Integer(
     A=&MaskIntBySpawnGroup(Masked Int=$Particles.PartitionParticles.Partition, Passthrough Int=0, Spawn Group Mask=2)
     B=&SelectIntFromArray(Direct Array Index=&SpawnGroup(), Int Selection Array=DataInterface() {propertyValues {IntData [0,1,4]}})
    )
    Bone / Socket Sampling Mode=::NewEnumerator5
    MaskSpawnGroup=false
    Mesh Sampling Type=::NewEnumerator6
    Orientation Sampling=::NewEnumerator1
    Sampled Position Offset=/VehicleCosmetics/FX/Niagara/Modules/MergeEmitters/VC_VectorBySpawnGroup(
     SpawnGroup or ExecIndex:NiagaraInt32=0
     CustomInt:NiagaraInt32=0
     MaximumSpawnGroups:NiagaraInt32=3
     Vector Group 0:Vector3f=<-28 0 -20>
     Vector Group 1:Vector3f=<-28 0 -20>
     Vector Group 2:Vector3f=<-40 0 -20>
     Vector Group 3:Vector3f=<0 0 0>
     Vector Group 4:Vector3f=<0 0 0>
     Vector Group 5:Vector3f=<0 0 0>
    )
    Sampled Position Offset Coordinate Space:ENiagaraSkelMeshTransforms=::NewEnumerator2
    Skeletal Mesh=$System.SkeletalMesh
   )
   &AddVelocity(Velocity=&VC_VectorBySpawnGroup(
     SpawnGroup or ExecIndex:NiagaraInt32=0
     CustomInt:NiagaraInt32=0
     MaximumSpawnGroups:NiagaraInt32=3
     Vector Group 0:Vector3f=&MakeCustomVectorFromBool(
      Bool=$User.bIsFrontendPreview
      True Vector=&RandomRangeVector(Minimum=<-450 0 0>, Maximum=<-650 0 0>)
      False Vector=<-1900 0 0>
     )
     Vector Group 1:Vector3f=&MakeCustomVectorFromBool(
      Bool=$User.bIsFrontendPreview
      True Vector=&RandomRangeVector(Minimum=<-450 0 0>, Maximum=<-650 0 0>)
      False Vector=<-1900 0 0>
     )
     Vector Group 2:Vector3f=<-500 0 0>
     Vector Group 3:Vector3f=<0 0 0>
     Vector Group 4:Vector3f=<0 0 0>
     Vector Group 5:Vector3f=<0 0 0>
    ))
   /VehicleCosmetics/FX/Niagara/Modules/MergeEmitters/VC_ExtraBindingColors(
    ExtraColors=1
    R1 Color:LinearColor=&MakeCustomLinearColorFromBool(Bool=$User.bIsPainted, True Color=$User.PaintedColor, False Color=<rgba 0.01 0.5 1 1>)
   )
  ]
  particleUpdateScript [
   &FN_DisableTireFX(
    Mask Index Attribute:NiagaraInt32=$Particles.PartitionParticles.Partition
    Mask Index FL=2
    Mask Index FR=3
    Mask Index RL=0
    Mask Index RR=1
   )
   &ParticleState()
   &ScaleSpriteSize(
    Scale Sprite Size Mode=::NewEnumerator0
    Uniform Scale Factor=&VC_FloatBySpawnGroup(
     SpawnGroup or ExecIndex:NiagaraInt32=0
     CustomInt:NiagaraInt32=0
     MaximumSpawnGroups:NiagaraInt32=3
     Float SG 0:NiagaraFloat=&FloatFromCurve(
      FloatCurve=DataInterface() {propertyValues {Curve {keys [<at 0 0.5 RCIM_Cubic>,<at 1 1.25 RCIM_Cubic>]}}}
      Scale Curve=&MakeCustomFloatFromBool(Bool=$User.bIsFrontendPreview, True Float=1, False Float=1.15)
     )
     Float SG 1:NiagaraFloat=&FloatFromCurve(
      FloatCurve=DataInterface() {propertyValues {Curve {keys [<at 0 0.5 RCIM_Cubic>,<at 1 1.25 RCIM_Cubic>]}}}
      Scale Curve=&MakeCustomFloatFromBool(Bool=$User.bIsFrontendPreview, True Float=1, False Float=1.15)
     )
     Float SG 2:NiagaraFloat=1
     Float SG 3:NiagaraFloat=0
     Float SG 4:NiagaraFloat=0
     Float SG 5:NiagaraFloat=0
    )
   ) {module=ScaleSpriteSize001}
   &InheritSourceMovement(
    Applied Force Scale=<1 1 0>
    Applied Position Scale=&VectorFromCurve(VectorCurve=DataInterface() {propertyValues {
       XCurve {keys [<at 0.3 1 RCIM_Cubic>,<at 1 0 RCIM_Cubic>]}
       YCurve {keys [<at 0.3 1 RCIM_Cubic>,<at 1 0 RCIM_Cubic>]}
       ZCurve {keys [<at 0.3 1 RCIM_Cubic>,<at 1 0 RCIM_Cubic>]}
      }})
    ApplyForce=false
    ApplyPosition=true
    ApplyPositionFalloff=false
   )
   &SolveForcesAndVelocity(Clamp Velocity=false, Limit Acceleration=false)
   &ScaleColor(
    Scale RGB=&MakeCustomVectorFromBool(
     Bool=$User.bIsPainted
     True Vector=&VectorFromCurve(
      VectorCurve=DataInterface() {propertyValues {
        XCurve {keys [{arriveTangent=-1.172, interpMode=RCIM_Cubic, leaveTangent=-1.172, tangentMode=RCTM_User, time=0, value=1}, <at 1 0.1 RCIM_Cubic>]}
        YCurve {keys [{arriveTangent=-1.172, interpMode=RCIM_Cubic, leaveTangent=-1.172, tangentMode=RCTM_User, time=0, value=1}, <at 1 0.1 RCIM_Cubic>]}
        ZCurve {keys [{arriveTangent=-1.172, interpMode=RCIM_Cubic, leaveTangent=-1.172, tangentMode=RCTM_User, time=0, value=1}, <at 1 0.1 RCIM_Cubic>]}
       }}
      Scale Curve=&Multiply_Vector(A=<1 1 1>, B=&VectorFromFloat(Value=$User.RingsBrightness))
     )
     False Vector=&VectorFromCurve(
      VectorCurve=DataInterface() {propertyValues {
        XCurve {keys [<at 0.25 1 RCIM_Cubic>,<at 1 0.1 RCIM_Cubic>]}
        YCurve {keys [<at 0.25 0.2 RCIM_Cubic>,<at 1 0.8 RCIM_Cubic>]}
        ZCurve {keys [<at 0.25 0.1 RCIM_Cubic>,<at 1 2 RCIM_Cubic>]}
       }}
      Scale Curve=<2.25 2.25 2.25>
     )
    )
    ScaleA=false
    ScaleRGB=true
   )
   &SubUVAnimation(
    "Sprite Renderer "=DataInterface() {propertyValues {SpriteRenderer=&self:NE_VC_Rings_Left001_0.NiagaraSpriteRendererProperties_2}}
    SubUV Animation Mode:ENiagara_SubUVLookupModeV2=::NewEnumerator1
    SubUV Curve=DataInterface() {propertyValues {Curve {keys [{arriveTangent=4.1785, interpMode=RCIM_Cubic, time=0, value=0}, {interpMode=RCIM_Cubic, leaveTangent=0.2086, time=1, value=1.14}]}}}
    UseEndFrame=false
    UseStartFrame=false
   )
   &SubUVAnimation(
    "Sprite Renderer "=DataInterface() {propertyValues {SpriteRenderer=&self:NE_VC_Rings_Left001_0.NiagaraSpriteRendererProperties_3}}
    SubUV Animation Mode:ENiagara_SubUVLookupModeV2=::NewEnumerator1
    SubUV Curve=DataInterface() {propertyValues {Curve {keys [{arriveTangent=4.1785, interpMode=RCIM_Cubic, time=0, value=0}, {interpMode=RCIM_Cubic, leaveTangent=0.2086, time=1, value=1.14}]}}}
    UseEndFrame=false
    UseStartFrame=false
   ) {module=SubUVAnimation001}
   &SubUVAnimation(
    Output Alternate Renderer Binding=true
    "Sprite Renderer "=DataInterface() {propertyValues {SpriteRenderer=&self:NE_VC_Rings_Left001_0.NiagaraSpriteRendererProperties_4}}
    SubUV Animation Mode:ENiagara_SubUVLookupModeV2=::NewEnumerator1
    SubUV Curve=DataInterface() {propertyValues {Curve {
       defaultValue=0
       keys [
        {arriveTangent=4.1785, interpMode=RCIM_Cubic, leaveTangent=4.1785, tangentMode=RCTM_User, time=0, value=0}
        {arriveTangent=1.1179, interpMode=RCIM_Cubic, leaveTangent=1.1179, tangentMode=RCTM_User, time=0.2042, value=0.6228}
        {arriveTangent=0.0224, interpMode=RCIM_Cubic, leaveTangent=0.0224, tangentMode=RCTM_User, time=1, value=1}
       ]
      }}}
    UseEndFrame=false
    UseStartFrame=false
   ) {module=SubUVAnimation002}
   &self:CustomFacing_02() {module=CustomFacing_02}
   &AlignSpriteToMeshOrientation(
    Mesh Orientation Relative Sprite Alignment Vector=<1 1 0>
    Mesh Orientation Relative Sprite Facing Vector:Vector3f=&MaskVectorBySpawnGroup(
     Masked Vector=&MakeCustomVectorFromBool(
      Bool=$User.bIsFrontendPreview
      True Vector=&MakeCustomVectorFromBool(Bool=$Particles.AgeTrigger, True Vector=$Particles.LocalDefaultFacing, False Vector=$Particles.LocalDefaultFacing)
      False Vector=&MakeCustomVectorFromBool(Bool=$Particles.AgeTrigger, True Vector=$Engine.Owner.Velocity, False Vector=$Particles.LocalDefaultFacing)
     )
     Passthrough Vector=&MaskVectorBySpawnGroup(
      Masked Vector=&MakeCustomVectorFromBool(
       Bool=$User.bIsFrontendPreview
       True Vector=&MakeCustomVectorFromBool(Bool=$Particles.AgeTrigger, True Vector=$Particles.LocalDefaultFacing, False Vector=$Particles.LocalDefaultFacing)
       False Vector=&MakeCustomVectorFromBool(Bool=$Particles.AgeTrigger, True Vector=$Engine.Owner.Velocity, False Vector=$Particles.LocalDefaultFacing)
      )
      Passthrough Vector=<1 1 1>
      Spawn Group Mask=1
     )
    )
    SetAlignment=false
    SetFacing=true
   ) {module=AlignSpriteToMeshOrientation001}
   &DynamicMaterialParameters(
    Index 0 Param 1=&MakeCustomFloatFromBool(Bool=$User.bIsPainted, True Float=$User.RingsBrightness, False Float=2.5)
    Index 1 Param 1=$User.DotsBrightness
    Index 1 Param 2=1
    Index 1 Param 3=1
    Module.Index0Param0Write=true
    Module.Index0Param1Write=true
    Module.Index0Param2Write=true
    Module.Index0Param3Write=true
    Module.Index1Param0Write=true
    Module.Index1Param1Write=true
    Module.Index1Param2Write=true
    Module.Index1Param3Write=true
    Write Parameter Index 1=true
   ) {enabled=false}
   /SetVariables(Particles.BlueDotsColor:LinearColor=&VC_DynamicScaleColor(
     Color to Scale:LinearColor=$Particles.R1 Color
     RGB Scale:Vector3f=&VectorFromCurve(VectorCurve=DataInterface() {propertyValues {
        XCurve {keys [<at 0 5 RCIM_Cubic>,<at 1 0 RCIM_Cubic>]}
        YCurve {keys [<at 0 5 RCIM_Cubic>,<at 1 0 RCIM_Cubic>]}
        ZCurve {keys [<at 0 5 RCIM_Cubic>,<at 1 0 RCIM_Cubic>]}
       }})
     Alpha Scale:NiagaraFloat=&FloatFromCurve(FloatCurve=DataInterface() {propertyValues {Curve {keys [<at 0 1 RCIM_Cubic>,<at 0.5 0 RCIM_Cubic>]}}})
    )) {isSetParameters=true}
  ]
  rendererClasses [&NiagaraSpriteRendererProperties]
  renderers [
   {rendererClass=&NiagaraSpriteRendererProperties}
   {rendererClass=&NiagaraSpriteRendererProperties, rendererIndex=1}
   {rendererClass=&NiagaraSpriteRendererProperties, rendererIndex=2}
  ]
 }
 NE_VC_ZigZag_Side001 {
  emitterUpdateScript [
   &EmitterState()
   &VC_RateQualityHandler(
    Epic Spawn Rate:NiagaraFloat=&MakeCustomFloatFromBool(Bool=$User.bIsFrontendPreview, True Float=55, False Float=40)
    High Spawn Rate=30
    Low Spawn Rate=30
    Use Custom Rate?=true
   )
   &VC_SpawnRibbon_v1(Number of Ribbons=2, SpawnRate=$Emitter.Spawn Rate)
  ]
  particleSpawnScript [
   &VC_PaintedVariantsOverride() {enabled=false}
   &VC_InitializeRibbon_v1(
    Color=&MakeCustomLinearColorFromBool(Bool=$User.bIsPainted, True Color=$User.PaintedColor, False Color=<rgba 1 1 1 1>)
    Custom Offset=true
    Enable Skeletal Position=true
    Facing Axis:ENiagaraOrientationAxis=::ZAxis
    Lifetime=&MakeCustomFloatFromBool(Bool=$User.bIsFrontendPreview, True Float=0.3, False Float=0.6)
    Local Vector Offset=&MakeCustomVectorFromBool(Bool=$User.bIsFrontendPreview, True Vector=<-23 0 -20>, False Vector=<-15 0 -20>)
    Offset Amount=1
    Ribbon Width=160
   )
   &AddVelocity(Velocity=&MakeCustomVectorFromBool(Bool=$User.bIsFrontendPreview, True Vector=<-4000 0 0>, False Vector=<-2200 0 0>)) {module=AddVelocity001}
   &VC_ExtraBindingColors(
    ExtraColors=1
    R1 Color:LinearColor=&MakeCustomLinearColorFromBool(Bool=$User.bIsPainted, True Color=$User.PaintedColor, False Color=<rgba 0.1 0.2 1 1>)
   )
  ]
  particleUpdateScript [
   &FN_DisableTireFX(Mask Index Attribute:NiagaraInt32=$Particles.SpawnGroup, Mask Index FL=2, Mask Index FR=3, Mask Index RL=0, Mask Index RR=1)
   &InheritSourceMovement(
    Applied Position Scale=&VectorFromCurve(VectorCurve=DataInterface() {propertyValues {
       XCurve {keys [<at 0.3 1 RCIM_Cubic>,<at 1 0 RCIM_Cubic>]}
       YCurve {keys [<at 0.3 1 RCIM_Cubic>,<at 1 0 RCIM_Cubic>]}
       ZCurve {keys [<at 0.3 1 RCIM_Cubic>,<at 1 0 RCIM_Cubic>]}
      }})
    ApplyForce=false
    ApplyPosition=true
    ApplyPositionFalloff=false
   )
   &ParticleState()
   &SolveForcesAndVelocity(Clamp Velocity=false, Limit Acceleration=false)
   &ScaleColor(
    Scale Alpha=&FloatFromCurve(FloatCurve=DataInterface() {propertyValues {Curve {keys [<at 0.8 1 RCIM_Cubic>,<at 1 0 RCIM_Cubic>]}}})
    Scale RGB=&MakeCustomVectorFromBool(
     Bool=$User.bIsPainted
     True Vector=&VectorFromCurve(
      VectorCurve=DataInterface() {propertyValues {
        XCurve {keys [
          <at 0 2 RCIM_Cubic>
          {arriveTangent=-1.8957, interpMode=RCIM_Cubic, leaveTangent=-1.8957, tangentMode=RCTM_User, time=0.1, value=1}
          <at 0.6 0.01 RCIM_Cubic>
         ]}
        YCurve {keys [
          <at 0 2 RCIM_Cubic>
          {arriveTangent=-1.8957, interpMode=RCIM_Cubic, leaveTangent=-1.8957, tangentMode=RCTM_User, time=0.1, value=1}
          <at 0.6 0.01 RCIM_Cubic>
         ]}
        ZCurve {keys [
          <at 0 2 RCIM_Cubic>
          {arriveTangent=-1.8957, interpMode=RCIM_Cubic, leaveTangent=-1.8957, tangentMode=RCTM_User, time=0.1, value=1}
          <at 0.6 0.01 RCIM_Cubic>
         ]}
       }}
      Scale Curve=<130 130 130>
     )
     False Vector=&VectorFromCurve(
      VectorCurve=DataInterface() {propertyValues {
        XCurve {keys [<at 0 5 RCIM_Cubic>, {arriveTangent=-4.95, leaveTangent=-4.95, time=0.2, value=0.1}, <at 0.825 0.005>]}
        YCurve {keys [
          <at 0 1 RCIM_Cubic>
          {arriveTangent=-0.0538, interpMode=RCIM_Cubic, leaveTangent=-0.0538, tangentMode=RCTM_User, time=0.2, value=0.5}
          {arriveTangent=-0.0441, interpMode=RCIM_Cubic, leaveTangent=-0.0441, tangentMode=RCTM_User, time=0.825, value=0.01}
         ]}
        ZCurve {keys [<at 0 0.05 RCIM_Cubic>, {arriveTangent=0.95, leaveTangent=0.95, time=0.2, value=1}, <at 0.825 0.1>]}
       }}
      Scale Curve=<30 30 30>
     )
    )
    ScaleA=false
    ScaleRGB=true
   )
   &DynamicMaterialParameters(
    Index 0 Param 1=&MakeCustomFloatFromBool(Bool=$User.bIsPainted, True Float=$User.RibbonBrightness, False Float=30)
    Index 0 Param 2=1
    Index 0 Param 3=1
    Index 0 Param 4=-7.5
    Index 1 Param 1=10
    Index 1 Param 2=1
    Index 1 Param 4=0.1
    Module.Index0Param0Write=true
    Module.Index0Param1Write=true
    Module.Index0Param2Write=true
    Module.Index0Param3Write=true
    Module.Index1Param0Write=true
    Module.Index1Param1Write=true
    Module.Index1Param2Write=true
    Module.Index1Param3Write=true
    Write Parameter Index 1=true
   ) {enabled=false}
   &VC_UpdateRibbon_v1(Enable Skeletal Position=true)
   /SetVariables(
    Particles.RibbonColor02:LinearColor=&VC_DynamicScaleColor(
     Color to Scale:LinearColor=$Particles.R1 Color
     RGB Scale:Vector3f=&VectorFromCurve(
      VectorCurve=DataInterface() {propertyValues {
        XCurve {keys [<at 0 10 RCIM_Cubic>,<at 0.15 1 RCIM_Cubic>]}
        YCurve {keys [<at 0 10 RCIM_Cubic>,<at 0.15 1 RCIM_Cubic>]}
        ZCurve {keys [<at 0 10 RCIM_Cubic>,<at 0.15 1 RCIM_Cubic>]}
       }}
      Scale Curve=<100 100 100>
     )
     Alpha Scale:NiagaraFloat=&FloatFromCurve(FloatCurve=DataInterface() {propertyValues {Curve {keys [<at 0 1 RCIM_Cubic>,<at 0.8 0 RCIM_Cubic>]}}})
    )
    Particles.RibbonWidth02:NiagaraFloat=&MakeCustomFloatFromBool(Bool=$User.bIsFrontendPreview, True Float=160, False Float=80)
    Particles.RibbonWidth03=60
   ) {isSetParameters=true}
  ]
  rendererClasses [&NiagaraRibbonRendererProperties]
  renderers [
   {rendererClass=&NiagaraRibbonRendererProperties}
   {rendererClass=&NiagaraRibbonRendererProperties, rendererIndex=1}
   {rendererClass=&NiagaraRibbonRendererProperties, rendererIndex=2}
  ]
 }
}
summary {
 systemName=NS_DriftTrail_ZigZag_MP
 emitters [
  {emitterName=NE_VC_ZigZag_Side001, rendererClasses [&NiagaraRibbonRendererProperties]}
  {emitterName=NE_VC_Rings_Left001, rendererClasses [&NiagaraSpriteRendererProperties]}
 ]
 userVariables [
  User.bIsFrontendPreview {type=NiagaraBool}
  User.bIsPainted {type=NiagaraBool}
  User.ColorVariantID {type=NiagaraInt32}
  User.DotsBrightness {type=NiagaraFloat, defaultValue=2, description=""}
  User.GameModeID {type=NiagaraInt32}
  User.offset {type=Vector3f, description="offset from wheels(set through cosmetic template)", defaultValue {struct=Vector3f, value=<-20 0 -20>}}
  User.PaintedColor {type=LinearColor, description="", defaultValue {struct=LinearColor, value=<rgba 1 1 1 1>}}
  User.PaintedScalar {type=NiagaraFloat, defaultValue=1, description=""}
  User.RibbonBrightness {type=NiagaraFloat, defaultValue=60, description=""}
  User.RingsBrightness {type=NiagaraFloat, defaultValue=10, description=""}
  User.VehicleFlags {type=NiagaraInt32, defaultValue=806289408}
  User.Wheel_Radius_Array {
   type=NiagaraDataInterfaceArrayFloat
   defaultValue {
    struct=NiagaraExt_VariableValue_DataInterface
    value {
     dataInterface=&self:NiagaraDataInterfaceArrayFloat_0
     dataInterfaceClass=&NiagaraDataInterfaceArrayFloat
    }
   }
  }
  User.Wheel_Width_Array {
   type=NiagaraDataInterfaceArrayFloat
   defaultValue {
    struct=NiagaraExt_VariableValue_DataInterface
    value {
     dataInterface=&self:NiagaraDataInterfaceArrayFloat_1
     dataInterfaceClass=&NiagaraDataInterfaceArrayFloat
    }
   }
  }
  User.WheelSize {type=NiagaraFloat, defaultValue=0.1, description=""}
 ]
}
